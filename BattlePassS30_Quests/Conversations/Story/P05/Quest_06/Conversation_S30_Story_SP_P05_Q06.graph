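# Conversation database for Season 30 story quest P05 Q06, where the player gives Hope the final Pandora Gem.
# A single entry point leads to Hope's gem request, an auto-selected main questline choice gated on the NPC's VoidRedemption entry tag, the quest being active and its obj0 objective being incomplete.
# The follow-up "give" choice carries the quest's gameplay tag and ends the conversation on Hope's reaction line.
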ConversationDatabase /BattlePassS30_Quests/Conversations/Story/P05/Quest_06/Conversation_S30_Story_SP_P05_Q06
entry_points [{tag=Conversation.Entry.Season30.Story.P05.Q06.VR, destinations [#1]}]
nodes [
 ConversationEntryPointNode_0 {
  id=#1
  class=ConversationEntryPointNode
  outputs [#2]
  properties {entryTag {tagName=Conversation.Entry.Season30.Story.P05.Q06.VR}}
 }
 Speech_C_1 {
  id=#2
  class=Speech_C
  outputs [#3]
  properties {softTaskIcon=&T_UI_NPC_MainQuest, generalConfig {defaultMessage=Okay. Give me the gem.}}
  sub_nodes [
   ConversationChoiceNode_2 {
    class=ConversationChoiceNode
    properties {
     defaultChoiceDisplayText=Talk to Hope
     choiceTags {gameplayTags [{tagName=Conversation.Choice.AutoSelectIfValid}, {tagName=Conversation.Choice.Category.MainQuestline}]}
    }
   }
   FortConversationRequirement_ControllerRequirement_1 {
    class=FortConversationRequirement_ControllerRequirement
    properties {
     nodeName=Require: NPC Has Conversation Tag
     participantId {tagName=Conversation.Participant.NPC}
     requirement {class=FortControllerRequirement_NPCHasConversationEntryTag, properties {queryTag {tagName=Conversation.Entry.VoidRedemption}}}
    }
   }
   FortConversationRequirement_ControllerRequirement_3 {
    class=FortConversationRequirement_ControllerRequirement
    properties {
     nodeName=Require: Is Quest Active
     requirement {
      class=FortControllerRequirement_IsQuestActive
      properties {questDefinition=/BattlePassS30_Quests/Items/QuestItems/Story/P05/Quest_S30_Story_P05_Q06}
     }
    }
   }
   FortConversationRequirement_ControllerRequirement_5 {
    class=FortConversationRequirement_ControllerRequirement
    properties {
     nodeName=Require: Has Completed Quest Objective
     requirement {
      class=FortControllerRequirement_HasCompletedQuestObjective
      properties {
       bInverseCondition=true
       objectiveBackendName=Quest_S30_Story_P05_Q06_obj0
       questDefinition=&Quest_S30_Story_P05_Q06
      }
     }
    }
   }
  ]
 }
 Speech_C_28 {
  id=#3
  class=Speech_C
  outputs []
  properties {softTaskIcon=&T_UI_NPC_Conversation_Give, generalConfig {defaultMessage=Whoa...}}
  sub_nodes [ConversationChoiceNode_1 {
    class=ConversationChoiceNode
    properties {
     defaultChoiceDisplayText=Give Hope the final Pandora Gem
     choiceTags {gameplayTags [{tagName=Athena.Quests.S30.Conversation.Story.P05.06.VR}]}
    }
   }]
 }
]
